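# Translucent unlit Niagara mesh-particle material that draws animated lightning for the Cactus Rocker back bling effects.
# World position offset combines SplineThicken width, tapered and varied by noise, with jitter from panning textures on rounded time for a stepped flicker.
# Lightning_Glow blends from the HDR-gradient bolt to a wider, soft-edged glow.
# Dynamic parameters drive width, offset UVs, jitter strength and emissive; particle color tints and fades the result.
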Material /BRCosmetics/Accessories/FORT_Backpacks/Backpack_F_MED_Cactus_Rocker/FX/Material/M_Cactus_Rocker_Lightning
graph {
 properties {
  bUsedWithNiagaraMeshParticles=true
  blendMode=BLEND_Translucent
  pixelDepthOffsetMode=PDOM_Legacy
  shadingModel=MSM_Unlit
  twoSided=true
  usedShadingModels=Unlit
 }
 nodes [
  n1 = TextureCoordinate() {uTiling=0, vTiling=0.3}
  n2 = DynamicParameter() {paramNames [Spline, WPO_UV, WPO_strength, Emissive_Strength]}
  n3 = Round(None=Multiply(A=Time()) {constB=24})
  n4 = TextureSample(UVs=Panner(Coordinate=Multiply(A=@n1, B=@n2.WPO_UV), Time=@n3) {bFractionalPart=true, speedX=0.1, speedY=5}) {samplerType=SAMPLERTYPE_Normal, texture=/Game/Global/Textures/Default/Tileable_Patterns/2ndInfectedMapNormal}
  n5 = Panner(
   Coordinate=Multiply(A=TextureCoordinate() {uTiling=0, vTiling=1.1}, B=DynamicParameter() {paramNames [Spline, WPO_UV, WPO_strength, Emissive_Strength]}.WPO_UV)
   Time=@n3
  ) {bFractionalPart=true, speedX=0.1, speedY=4}
  n6 = LinearInterpolate(
   A=Power(Base=Multiply(A=@n4.R) {constB=1.5}) {constExponent=4}
   B=LinearInterpolate(Alpha=TextureSample(UVs=@n5) {texture=/Game/CharClasses/Outlander/Gadgets/AntiMaterialCharge/Materials/T_AntiMaterialCharge_Dissolve_01}.R) {constA=-1}
  )
  n7 = MaterialFunctionCall(X=@n6, Y=@n6, Z=@n6) {materialFunction=&MakeFloat3}
  n8 = DynamicParameter() {paramNames [Spline, WPO_UV, WPO_strength, Emissive_Strength]}
  n9 = Time()
  n10 = Panner(
   Coordinate=Multiply(A=TextureCoordinate() {uTiling=0, vTiling=0.2}, B=DynamicParameter() {paramNames [Spline, WPO_UV, WPO_strength, Emissive_Strength]}.WPO_UV)
   Time=Round(None=Multiply(A=@n9) {constB=28})
  ) {bFractionalPart=true, speedY=0.7}
  n11 = LinearInterpolate(Alpha=Power(Base=TextureSample(UVs=@n10) {samplerType=SAMPLERTYPE_Normal, texture=&2ndInfectedMapNormal}.G)) {constA=0.52, constB=6.8}
  n12 = ComponentMask(None=TextureCoordinate()) {g=true}
  n13 = OneMinus(None=@n12)
  n14 = LinearInterpolate(B=Multiply(A=Constant() {r=1.5}, B=DynamicParameter() {paramNames [Spline, WPO_UV, WPO_strength, Emissive_Strength]}.Spline), Alpha=@n13) {constA=0.35, constB=1.5}
  n15 = LinearInterpolate(B=Multiply(A=Constant() {r=1.5}, B=DynamicParameter() {paramNames [Spline, WPO_UV, WPO_strength, Emissive_Strength]}.Spline), Alpha=@n12) {constA=0.35, constB=1.8}
  n16 = Add(
   A=Transform(None=Multiply(A=@n7.Result, B=@n8.WPO_strength) {constB=1.55}) {transformType=TRANSFORM_View}
   B=MaterialFunctionCall(WidthBase=Multiply(A=@n11, B=@n14), WidthTip=Multiply(A=@n11, B=@n15)) {materialFunction=&SplineThicken}.WorldPositionOffset
  )
  n17 = ScalarParameter() {defaultValue=6, parameterName=GlowSpline}
  n18 = LinearInterpolate(B=@n17, Alpha=@n13) {constA=0.15}
  n19 = LinearInterpolate(B=@n17, Alpha=@n12) {constA=0.15}
  LinearInterpolate(
   A=@n16
   B=Add(
    A=@n16
    B=MaterialFunctionCall(WidthBase=@n18, WidthTip=@n19) {materialFunction=&SplineThicken}.WorldPositionOffset
   )
   Alpha=ScalarParameter() {parameterName=Lightning_Glow}
  )
  n20 = DynamicParameter() {paramNames [Spline, WPO_UV, WPO_strength, Emissive_Strength]}
  n21 = TextureCoordinate()
  n22 = Power(Base=SphereMask(A=ComponentMask(None=@n21) {r=true}, B=Constant() {r=0.5}) {attenuationRadius=0.39, hardnessPercent=0}) {constExponent=2.3}
  n23 = LinearInterpolate(
   A=VectorParameter() {parameterName=Color_Dark, defaultValue=<rgba 0.1458 0 3 1>}.RGB
   B=VectorParameter() {parameterName=Color_Light, defaultValue=<rgba 0.9857 1.7688 300 1>}.RGB
   Alpha=@n22
  )
  n24 = Multiply(
   A=@n20.Emissive_Strength
   B=MaterialFunctionCall(LDR Input=Clamp(None=@n23), HDR Tint=@n23, HDR Intensity Multiplier=LinearInterpolate(Alpha=@n22) {constA=0.9}) {materialFunction=&DeriveHDRfromLDR}.RemappedRange
  )
  n25 = Multiply(A=@n20.Emissive_Strength, B=VectorParameter() {parameterName=GlowColor_Light, defaultValue=<rgba 0 122.9656 300 1>}.RGB)
  n26 = ScalarParameter() {parameterName=Lightning_Glow}
  MaterialFunctionCall(Emissive=Multiply(A=LinearInterpolate(A=@n24, B=@n25, Alpha=@n26), B=ParticleColor().RGB)) {materialFunction=&MF_FX_EmissiveAdjustments}
  n27 = Constant() {r=30}
  n28 = TextureCoordinate()
  n29 = Power(Base=SphereMask(A=ComponentMask(None=@n28) {r=true}, B=Constant() {r=0.5}) {attenuationRadius=0.5, hardnessPercent=0})
  n30 = ScalarParameter() {defaultValue=1, parameterName=Glow_Opacity}
  Multiply(
   A=LinearInterpolate(A=DepthFade(Opacity=@n27), B=Multiply(A=@n29, B=@n30), Alpha=ScalarParameter() {parameterName=Lightning_Glow})
   B=ParticleColor().A
  )
 ]
}
