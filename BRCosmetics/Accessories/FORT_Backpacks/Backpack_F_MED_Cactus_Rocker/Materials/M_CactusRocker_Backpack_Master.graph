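# Master material for the Cactus Rocker back bling, drawing audio-reactive spectrum bars on a skeletal mesh.
# A four-channel spectrum from SampleAudioMPCHistory, or fake parameters via static switch, expands into seven bands selected by a band-ID mask.
# A curve atlas colors base and emissive, alongside sine-wave gradient modulation, Fresnel falloff and two bump-offset sparkle layers driven by light and camera direction.
# Standard character skin and post-FX functions wrap the result.
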Material /BRCosmetics/Accessories/FORT_Backpacks/Backpack_F_MED_Cactus_Rocker/Materials/M_CactusRocker_Backpack_Master
graph {
 properties {
  bUseMaterialAttributes=true
  bUsedWithMeshDeformer=true
  bUsedWithSkeletalMesh=true
  pixelDepthOffsetMode=PDOM_Legacy
  refractionMethod=RM_IndexOfRefraction
  usedShadingModels=Default Lit
 }
 nodes [
  Ceil(None=NamedRerouteUsage())
  n1 = MaterialFunctionCall(TexCoords=TextureCoordinate()) {materialFunction=&BaseFN_CharacterFunction_PreFX}
  n2 = SetMaterialAttributes!(
   MaterialAttributes=MaterialFunctionCall(
    PreskinnedPosition=@n1.Pre skinned Local Position
    Gmap=@n1.Pre skinned Local Position
    MaterialAttributes=@n1.Pre skinned Local Position
    AO=@n1.Pre skinned Local Position
   ) {materialFunction=&CharacterSkinFX}.Result
   Base Color=NamedRerouteUsage()
   Emissive Color=NamedRerouteUsage()
   Metallic=ScalarParameter() {defaultValue=0.5, group=Cactus Rocker, parameterName=Metallic}
   Specular=ScalarParameter() {defaultValue=1, group=Cactus Rocker, parameterName=Specular}
   Roughness=ScalarParameter() {defaultValue=0.1, group=Cactus Rocker, parameterName=Roughness}
  ) {attributeSetTypes [
    #1
    #2
    #3
    #4
    #5
   ]}
  n3 = MaterialFunctionCall(AO=@n1.AO, Distance Blend=@n1.AO, Fresnel=@n1.AO, Gmap=@n1.AO, Pre Skinned Local Position=@n1.AO, Material=@n2) {materialFunction=&BaseFN_CharacterFunction_PostFX}
  MaterialFunctionCall(In=@n3.Result) {materialFunction=&CINE_MF_CharPostFX}
  n4 = Divide(A=VertexInterpolator(VS=PreSkinnedPosition()).PS, B=PreSkinnedLocalBounds().Half Extents)
  n5 = OneMinus(None=Saturate(None=Abs(None=@n4)))
  n6 = Add(A=Time(), B=NamedRerouteDeclaration(None=ComponentMask(None=@n5) {g=true}) {name=YPos, nodeColor=<rgba 0 1 0.0156 1>})
  n7 = ScalarParameter() {defaultValue=0.5, parameterName=WaveWidth}
  n8 = Multiply(A=Sine(None=Divide(A=@n6, B=@n7)), B=ScalarParameter() {defaultValue=0.025, parameterName=WaveHeight})
  n9 = TextureSample() {
   samplerSource=SSM_Wrap_WorldGroupSettings
   samplerType=SAMPLERTYPE_LinearGrayscale
   texture=/BRCosmetics/Accessories/FORT_Backpacks/Backpack_F_MED_Cactus_Rocker/Textures/T_CactusGradient_color
  }
  NamedRerouteDeclaration(None=Saturate(None=Add(A=@n8, B=@n9.G))) {name=Modulated_ZPos, nodeColor=<rgba 1 0 0.8672 1>}
  n10 = MaterialFunctionCall(Float4=Ceil(None=NamedRerouteUsage())) {materialFunction=&BreakOutFloat4Components}
  n11 = Add(A=Add(A=@n10.R, B=@n10.R), B=@n10.B)
  NamedRerouteDeclaration(None=Saturate(None=Add(A=@n11, B=@n10.A))) {name=Is_Audio_Active_Mask, nodeColor=<rgba 1 0 0.3984 1>}
  n12 = TextureSample() {
   samplerSource=SSM_Wrap_WorldGroupSettings
   samplerType=SAMPLERTYPE_Masks
   texture=/BRCosmetics/Accessories/FORT_Backpacks/Backpack_F_MED_Cactus_Rocker/Textures/T_F_MED_Cactus_Rocker_Pack_Backpack_FX
  }
  NamedRerouteDeclaration(None=@n12.G) {name=Band_GlowMask, nodeColor=<rgba 1 0 0.8437 1>}
  n13 = NamedRerouteUsage()
  n14 = Abs(None=ConstantBiasScale(None=@n13) {bias=-0.5, scale=2})
  NamedRerouteDeclaration(None=OneMinus(None=StaticSwitchParameter(True=@n14, False=@n13) {parameterName=Use_Center_Based_Spectrum})) {name=Spectrum_Pos, nodeColor=<rgba 0 1 0.8359 1>}
  n15 = NamedRerouteUsage()
  n16 = Saturate(None=Min(A=OneMinus(None=@n15)) {constB=0.99})
  n17 = ComponentMask(None=@n16) {r=true}
  n18 = Subtract(A=NamedRerouteUsage(), B=@n17)
  n19 = Constant() {r=0.25}
  NamedRerouteDeclaration(None=Saturate(None=MaterialFunctionCall!(In=@n18, Ramp Width=@n19) {materialFunction=/Engine/Functions/Engine_MaterialFunctions01/Math/SmoothCeil}.Result) {desc=Band 1}) {name=Band_1, nodeColor=<rgba 1 0 0.4922 1>}
  n20 = ComponentMask(None=@n16) {g=true}
  n21 = Add(A=@n17, B=@n20)
  n22 = Subtract(A=NamedRerouteUsage(), B=Power(Base=Multiply(A=@n21) {constB=0.5}) {constExponent=0.7})
  n23 = Constant() {r=0.25}
  NamedRerouteDeclaration(None=Saturate(None=MaterialFunctionCall(In=@n22, Ramp Width=@n23) {materialFunction=&SmoothCeil}.Result) {desc="Band  2"}) {name=Band_2, nodeColor=<rgba 1 0 0.4922 1>}
  n24 = Subtract(A=NamedRerouteUsage(), B=@n20)
  n25 = Constant() {r=0.25}
  NamedRerouteDeclaration(None=Saturate(None=MaterialFunctionCall(In=@n24, Ramp Width=@n25) {materialFunction=&SmoothCeil}.Result) {desc=Band 3}) {name=Band_3, nodeColor=<rgba 1 0 0.4922 1>}
  n26 = ComponentMask(None=@n16) {b=true}
  n27 = Add(A=@n20, B=@n26)
  n28 = Subtract(A=NamedRerouteUsage(), B=Power(Base=Multiply(A=@n27) {constB=0.5}) {constExponent=0.7})
  n29 = Constant() {r=0.25}
  NamedRerouteDeclaration(None=Saturate(None=MaterialFunctionCall(In=@n28, Ramp Width=@n29) {materialFunction=&SmoothCeil}.Result) {desc="Band  4"}) {name=Band_4, nodeColor=<rgba 1 0 0.4922 1>}
  n30 = Subtract(A=NamedRerouteUsage(), B=@n26)
  n31 = Constant() {r=0.25}
  NamedRerouteDeclaration(None=Saturate(None=MaterialFunctionCall(In=@n30, Ramp Width=@n31) {materialFunction=&SmoothCeil}.Result) {desc=Band 5}) {name=Band_5, nodeColor=<rgba 1 0 0.4922 1>}
  n32 = ComponentMask(None=@n16) {a=true}
  n33 = Add(A=@n26, B=@n32)
  n34 = Subtract(A=NamedRerouteUsage(), B=Power(Base=Multiply(A=@n33) {constB=0.5}) {constExponent=0.7})
  n35 = Constant() {r=0.25}
  NamedRerouteDeclaration(None=Saturate(None=MaterialFunctionCall(In=@n34, Ramp Width=@n35) {materialFunction=&SmoothCeil}.Result) {desc="Band  6"}) {name=Band_6, nodeColor=<rgba 1 0 0.4922 1>}
  n36 = Subtract(A=NamedRerouteUsage(), B=@n32)
  n37 = Constant() {r=0.25}
  NamedRerouteDeclaration(None=Saturate(None=MaterialFunctionCall(In=@n36, Ramp Width=@n37) {materialFunction=&SmoothCeil}.Result) {desc=Band 7}) {name=Band_7, nodeColor=<rgba 1 0 0.4922 1>}
  n38 = Ceil(None=Subtract(A=NamedRerouteUsage()) {constB=0.1})
  n39 = Ceil(None=Subtract(A=NamedRerouteUsage()) {constB=0.25})
  n40 = OneMinus(None=@n39)
  n41 = Ceil(None=Subtract(A=NamedRerouteUsage()) {constB=0.4})
  n42 = OneMinus(None=@n41)
  n43 = Add(A=Multiply(A=Multiply(A=@n38, B=@n40), B=NamedRerouteUsage()), B=Multiply(A=Multiply(A=@n39, B=@n42), B=NamedRerouteUsage()))
  n44 = Ceil(None=Subtract(A=NamedRerouteUsage()) {constB=0.5})
  n45 = Multiply(A=@n41, B=OneMinus(None=@n44))
  n46 = NamedRerouteUsage()
  n47 = Ceil(None=Subtract(A=NamedRerouteUsage()) {constB=0.65})
  n48 = Multiply(A=@n44, B=OneMinus(None=@n47))
  n49 = NamedRerouteUsage()
  n50 = Ceil(None=Subtract(A=NamedRerouteUsage()) {constB=0.8})
  n51 = Multiply(A=Multiply(A=@n47, B=OneMinus(None=@n50)), B=NamedRerouteUsage())
  n52 = Ceil(None=Subtract(A=NamedRerouteUsage()) {constB=0.95})
  n53 = Multiply(A=Multiply(A=@n50, B=OneMinus(None=@n52)), B=NamedRerouteUsage())
  n54 = Saturate(None=@n52)
  n55 = NamedRerouteUsage()
  n56 = Add(A=Add(A=Multiply(A=@n45, B=@n46), B=Multiply(A=@n48, B=@n49)), B=Add(A=Add(A=@n51, B=@n53), B=Multiply(A=@n54, B=@n55)))
  n57 = Multiply(A=Ceil(None=NamedRerouteUsage()), B=Multiply(A=Add(A=@n43, B=@n56), B=NamedRerouteUsage()))
  n58 = NamedRerouteUsage()
  NamedRerouteDeclaration(None=Saturate(None=Multiply(A=@n57, B=@n58))) {name=Spectrum_Bars_Mask, nodeColor=<rgba 1 0.6563 0 1>}
  n59 = VectorParameter() {parameterName=Fake_Audio_Bands, defaultValue=<rgba 0.5 0.9 0.1 0.75>}
  n60 = AppendVector(A=@n59.RGB, B=@n59.RGB)
  n61 = MaterialFunctionCall!(useSmoothStep=StaticBool(), uvx=Constant() {r=0.5}) {materialFunction=/Game/Accessories/FORT_Backpacks/Backpack_F_MED_Cactus_Rocker/Materials/SampleAudioMPCHistory}
  NamedRerouteDeclaration(None=StaticSwitchParameter(True=OneMinus(None=@n60), False=@n61.Interpolated Spectrum) {group=Cactus Rocker, parameterName=UseFakeAudioData}) {name=Audio_Spectrum_In, nodeColor=<rgba 1 0 0.1875 1>}
  n62 = Ceil(None=NamedRerouteUsage())
  n63 = Saturate(None=Power(Base=NamedRerouteUsage()))
  NamedRerouteDeclaration(None=Fresnel(
    ExponentIn=LinearInterpolate(Alpha=@n62) {constA=1.5, constB=3}
    BaseReflectFractionIn=LinearInterpolate(Alpha=@n63) {constA=0.35, constB=0.175}
   ) {baseReflectFraction=0.2, exponent=1}) {name=Edge_Falloff_Mask, nodeColor=<rgba 0.3359 0 1 1>}
  n64 = CurveAtlasRowParameter(CurveTime=NamedRerouteUsage()) {
   atlas=/Game/Accessories/FORT_Backpacks/Backpack_F_MED_Cactus_Rocker/Textures/CA_CactusRocker
   curve=/Game/Accessories/FORT_Backpacks/Backpack_F_MED_Cactus_Rocker/Textures/CactusRockerColor
   parameterName=CA_Parameter
  }
  n65 = Desaturation(None=@n64, Fraction=Constant() {r=-1})
  n66 = Multiply(A=LinearInterpolate(A=@n64, B=@n65, Alpha=NamedRerouteUsage()), B=NamedRerouteUsage())
  n67 = Multiply(A=NamedRerouteUsage(), B=@n65)
  n68 = Multiply(A=@n67, B=ScalarParameter() {defaultValue=3, parameterName=Spectrum_Emissive_Intensity})
  n69 = NamedRerouteUsage()
  NamedRerouteDeclaration(None=Add(A=LinearInterpolate(A=@n66, B=@n68, Alpha=@n69), B=NamedRerouteUsage())) {name=Emissive_Out, nodeColor=<rgba 1 0 0.1172 1>}
  n70 = NamedRerouteUsage()
  n71 = NamedRerouteUsage()
  NamedRerouteDeclaration(None=Multiply(A=Multiply(A=@n64, B=@n70), B=LinearInterpolate(A=@n66, B=@n67, Alpha=@n71))) {name=BaseColor_Out, nodeColor=<rgba 0.2187 1 0 1>}
  n72 = NamedRerouteUsage()
  n73 = TextureCoordinate()
  n74 = ScalarParameter() {defaultValue=2, parameterName=SparkleTiling}
  n75 = TextureSample(UVs=BumpOffset(
    Coordinate=Multiply(A=@n73, B=@n74)
    Height=ScalarParameter() {parameterName="SparkleCrackBump\\"}
    HeightRatioInput=ScalarParameter() {parameterName=SparkleHeightRatio}
   ) {heightRatio=1, referencePlane=0}) {samplerType=SAMPLERTYPE_Masks, texture=/Game/Weapons/WeaponSkins/Textures/T_SparkleMask}
  n76 = ConstantBiasScale(None=@n75.R)
  n77 = AtmosphericLightVector()
  n78 = Constant3Vector() {constant=<rgba 1 0 0 0>}
  n79 = MaterialFunctionCall() {materialFunction=&CameraDirectionVector}
  n80 = Add(A=ConstantBiasScale(None=DotProduct(A=@n77, B=@n78)), B=ConstantBiasScale(None=DotProduct(A=@n79.Result, B=@n78)))
  n81 = Frac(None=Add(A=Multiply(A=@n76, B=@n75.G), B=@n80))
  n82 = Power(
   Base=Multiply(
    A=@n75.R
    B=MaterialFunctionCall(UV Channel=@n81) {materialFunction=&LinearGradient}.VGradient
   )
   Exp=ScalarParameter() {defaultValue=12, parameterName=Layer_01SparkleSharpness}
  ) {constExponent=12}
  n83 = ScalarParameter() {defaultValue=10, parameterName=Layer_01SparkleBrightness}
  n84 = Add(
   A=TextureCoordinate()
   B=AppendVector(
    A=ScalarParameter() {defaultValue=0.25, parameterName=Sparkle_Layer02_Offset_X}
    B=ScalarParameter() {defaultValue=0.5, parameterName=Sparkle_Layer02_Offset_Y}
   )
  )
  n85 = ScalarParameter() {defaultValue=2, parameterName=SparkleTiling_02}
  n86 = TextureSample(UVs=BumpOffset(
    Coordinate=Multiply(A=@n84, B=@n85)
    Height=ScalarParameter() {parameterName=SparkleCrackBump_02}
    HeightRatioInput=ScalarParameter() {defaultValue=0.1, parameterName=SparkleHeightRatio_02}
   ) {heightRatio=1, referencePlane=0.1}) {samplerType=SAMPLERTYPE_Masks, texture=&T_SparkleMask}
  n87 = ConstantBiasScale(None=@n86.R)
  n88 = Frac(None=Add(A=Multiply(A=@n87, B=@n86.G), B=@n80))
  n89 = Power(
   Base=Multiply(
    A=@n86.R
    B=MaterialFunctionCall(UV Channel=@n88) {materialFunction=&LinearGradient}.VGradient
   )
   Exp=ScalarParameter() {defaultValue=12, parameterName=Layer_02SparkleSharpness}
  ) {constExponent=12}
  n90 = ScalarParameter() {defaultValue=50, parameterName=Layer_02SparkleBrightness}
  n91 = Add(A=Saturate(None=Multiply(A=@n82, B=@n83) {constB=30}), B=Saturate(None=Multiply(A=@n89, B=@n90) {constB=30})) {constB=0}
  n92 = NamedRerouteUsage()
  NamedRerouteDeclaration(None=Multiply(A=LinearInterpolate(Alpha=@n72) {constA=0.75, constB=1.33}, B=Multiply(A=@n91, B=@n92))) {name=Sparkles, nodeColor=<rgba 1 0.5625 0 1>}
  NamedRerouteDeclaration(None=@n12.B) {name=Thorn_Mask, nodeColor=<rgba 0 0.5781 1 1>}
  NamedRerouteDeclaration(None=@n12.R) {name=Spectrum_Band_IDs, nodeColor=<rgba 1 0 0.1406 1>}
  NamedRerouteDeclaration(None=StaticSwitchParameter(True=ScalarParameter() {defaultValue=1, group=Cactus Rocker, parameterName=Fake_Audio_Amplitude}, False=@n61.Interpolated Amplitude Music) {group=Cactus Rocker, parameterName=UseFakeAudioData}) {name=Audio_Amplitude_In, nodeColor=<rgba 1 0 0.8437 1>}
 ]
}
