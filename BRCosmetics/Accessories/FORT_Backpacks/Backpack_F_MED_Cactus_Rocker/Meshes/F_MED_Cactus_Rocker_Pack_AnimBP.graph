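# Animation Blueprint for the Cactus Rocker backpack that adds arm sway and adjusts the cactus pose during skydiving.
# Two chained Anim Dynamics nodes simulate the right and left arm bones from the reference pose, then a Modify Bone node sets the master cactus bone's yaw.
# Each update reads bIsSkyDiving: skydiving zeroes the dynamics alpha and yaws the cactus -15 degrees, otherwise sway is full and yaw is zero.
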AnimBlueprint /BRCosmetics/Accessories/FORT_Backpacks/Backpack_F_MED_Cactus_Rocker/Meshes/F_MED_Cactus_Rocker_Pack_AnimBP
target_skeleton=&/F_MED_Cactus_Rocker_Pack_Skeleton
graphs [
 AnimGraph {nodes [
   n1 = Get skydiving_value() -> (skydiving_value:double)
   n2 = Anim Dynamics - Bone: right_arm_dyn(ComponentPose=Local To Component(LocalPose=Local Space Ref Pose()), Alpha=@n1, ExternalForce="0,0,0") {
    angularDampingOverride=0.99
    bOverrideAngularDamping=true
    alphaBoolBlend {blendOption=Linear}
    boundBone {boneName=right_arm_dyn}
    physicsBodyDefinitions [{
      sphereCollisionRadius=0
      boundBone {boneName=right_arm_dyn}
      boxExtents=<80 80 80>
      constraintSetup {angularLimitsMax=<15 15 15>, angularLimitsMin=<-15 -15 -15>, angularTarget=<0 0 0>}
      localJointOffset=<20 0 0>
     }]
    relativeSpaceBone {boneName=root}
   }
   n3 = Anim Dynamics - Bone: left_arm_dyn(ComponentPose=@n2, Alpha=@n1, ExternalForce="0,0,0") {
    angularDampingOverride=0.99
    bOverrideAngularDamping=true
    alphaBoolBlend {blendOption=Linear}
    boundBone {boneName=left_arm_dyn}
    physicsBodyDefinitions [{
      sphereCollisionRadius=0
      boundBone {boneName=left_arm_dyn}
      boxExtents=<60 60 60>
      constraintSetup {angularLimitsMax=<15 15 15>, angularLimitsMin=<-15 -15 -15>, angularTarget=<0 0 0>}
      localJointOffset=<20 0 0>
     }]
    relativeSpaceBone {boneName=root}
   }
   n4 = "Transform (Modify) Bone - Bone: cactus_mstr"!(
    ComponentPose=@n3
    Alpha=1
    Translation="0,0,0"
    Rotation_Roll:float=0
    Rotation_Pitch:float=0
    Rotation_Yaw:float=Get skydiving_rotation() -> (skydiving_rotation:double)
    Scale="1,1,1"
   ) -> (Pose) {rotationMode=BMM_Replace, rotationSpace=BCS_BoneSpace, alphaBoolBlend {blendOption=Linear}, boneToModify {boneName=cactus_mstr}}
   Output Pose(Result=Component To Local(ComponentPose=@n4)) {layerGroup=None}
  ]}
 EventGraph {nodes [
   Event BlueprintUpdateAnimation()
   n1 = Get bIsSkyDiving()
   > Set skydiving_value(skydiving_value:double=SelectFloat(A=0, B=1, bPickA=@n1)) -> (Output_Get:double)
   > Set skydiving_rotation(skydiving_rotation:double=SelectFloat(A=-15, B=0, bPickA=@n1)) -> (Output_Get:double)
   TryGetPawnOwner()
  ]}
]
variables [
 skydiving_rotation {type=number}
 skydiving_value {type=number}
]
